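# Sound cue for third-person soldier landings on metal heard from below in Athena, using "Downstairs" wave variants and the enemy foley sound class.
# A quality-level switch sends its first two levels to a random pick of six waves and its last two to a three-wave subset, each through pitch and volume modulation.
# Landing attenuation applies, with footstep concurrency plus a per-owner override that stops the oldest sound.
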SoundCue /Game/Athena/Sounds/Character/Footsteps/3P/Metal/Athena_Soldier_Metal_Land_Below_3P_Cue
root=QualityLevel_0
nodes [
 {id=Modulator_1, children [Random_3], properties {pitchMax=1.1, pitchMin=0.9, volumeMax=1.3, volumeMin=0.7}}
 {id=Modulator_5, children [Random_1], properties {pitchMax=1.1, pitchMin=0.9, volumeMax=1.3, volumeMin=0.7}}
 {id=QualityLevel_0, children [Modulator_5,Modulator_5,Modulator_1,Modulator_1]}
 {
  id=Random_1
  children [WavePlayer_17, WavePlayer_18, WavePlayer_19, WavePlayer_20, WavePlayer_21, WavePlayer_22]
  properties {weights [1,1,1,1,1,1]}
 }
 {id=Random_3, children [WavePlayer_9, WavePlayer_10, WavePlayer_11], properties {bShouldExcludeFromBranchCulling=true, weights [1,1,1]}}
 {
  id=WavePlayer_10
  properties {soundWaveAssetPtr=/Game/Sounds/Fort_Step_Sounds/Soldier/Elevations/Downstairs/Soldier_Step_Land_Metal_02_Downstairs}
 }
 {
  id=WavePlayer_11
  properties {soundWaveAssetPtr=/Game/Sounds/Fort_Step_Sounds/Soldier/Elevations/Downstairs/Soldier_Step_Land_Metal_03_Downstairs}
 }
 {
  id=WavePlayer_17
  properties {soundWaveAssetPtr=/Game/Sounds/Fort_Step_Sounds/Soldier/Elevations/Downstairs/Soldier_Step_Land_Metal_01_Downstairs}
 }
 {
  id=WavePlayer_18
  properties {soundWaveAssetPtr=&Soldier_Step_Land_Metal_02_Downstairs}
 }
 {
  id=WavePlayer_19
  properties {soundWaveAssetPtr=&Soldier_Step_Land_Metal_03_Downstairs}
 }
 {
  id=WavePlayer_20
  properties {soundWaveAssetPtr=/Game/Sounds/Fort_Step_Sounds/Soldier/Elevations/Downstairs/Soldier_Step_Land_Metal_04_Downstairs}
 }
 {
  id=WavePlayer_21
  properties {soundWaveAssetPtr=/Game/Sounds/Fort_Step_Sounds/Soldier/Elevations/Downstairs/Soldier_Step_Land_Metal_05_Downstairs}
 }
 {
  id=WavePlayer_22
  properties {soundWaveAssetPtr=/Game/Sounds/Fort_Step_Sounds/Soldier/Elevations/Downstairs/Soldier_Step_Land_Metal_06_Downstairs}
 }
 {
  id=WavePlayer_9
  properties {soundWaveAssetPtr=&Soldier_Step_Land_Metal_01_Downstairs}
 }
]
properties {
 attenuationSettings=/Game/Sounds/Attenuation/Fort_Player_Land_3P_Athena_Att
 maxDistance=3584
 soundClassObject=&Fort_Player_Foley_3P_Enemy_Athena
 volumeMultiplier=0.9
 concurrencyOverrides {bLimitToOwner=true, maxCount=4, resolutionRule=StopOldest}
 concurrencySet [&Athena_Footstep_Global_Conc]
}
