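# Sound cue for third-person enemy soldier sprint footsteps on metal heard from above, built from "Upstairs" walk samples.
# A quality-level switch selects a ten-wave random pool for the first two levels or a three-wave subset, excluded from branch culling, for the other two.
# A root modulator boosts volume, with sprint-specific attenuation and local and global footstep concurrency applied.
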SoundCue /Game/Athena/Sounds/Character/Footsteps/3P/Metal/Athena_Soldier_Metal_Sprint_Footstep_Above_3P_Cue
root=Modulator_3
nodes [
 {id=Modulator_3, children [QualityLevel_1], properties {volumeMax=1.5, volumeMin=1.5}}
 {id=QualityLevel_1, children [Random_1,Random_1,Random_4,Random_4]}
 {
  id=Random_1
  children [
   WavePlayer_12
   WavePlayer_13
   WavePlayer_14
   WavePlayer_22
   WavePlayer_23
   WavePlayer_24
   WavePlayer_25
   WavePlayer_26
   WavePlayer_27
   WavePlayer_28
  ]
  properties {weights [1,1,1,1,1,1,1,1,1,1]}
 }
 {id=Random_4, children [WavePlayer_34, WavePlayer_35, WavePlayer_36], properties {bShouldExcludeFromBranchCulling=true, weights [1,1,1]}}
 {
  id=WavePlayer_12
  properties {soundWaveAssetPtr=/Game/Sounds/Fort_Step_Sounds/Soldier/Elevations/Upstairs/Soldier_Step_Walk_Metal_04_Upstairs2}
 }
 {
  id=WavePlayer_13
  properties {soundWaveAssetPtr=/Game/Sounds/Fort_Step_Sounds/Soldier/Elevations/Upstairs/Soldier_Step_Walk_Metal_11_Upstairs2}
 }
 {
  id=WavePlayer_14
  properties {soundWaveAssetPtr=/Game/Sounds/Fort_Step_Sounds/Soldier/Elevations/Upstairs/Soldier_Step_Walk_Metal_01_Upstairs2}
 }
 {
  id=WavePlayer_22
  properties {soundWaveAssetPtr=/Game/Sounds/Fort_Step_Sounds/Soldier/Elevations/Upstairs/Soldier_Step_Walk_Metal_02_Upstairs2}
 }
 {
  id=WavePlayer_23
  properties {soundWaveAssetPtr=/Game/Sounds/Fort_Step_Sounds/Soldier/Elevations/Upstairs/Soldier_Step_Walk_Metal_03_Upstairs2}
 }
 {
  id=WavePlayer_24
  properties {soundWaveAssetPtr=/Game/Sounds/Fort_Step_Sounds/Soldier/Elevations/Upstairs/Soldier_Step_Walk_Metal_05_Upstairs2}
 }
 {
  id=WavePlayer_25
  properties {soundWaveAssetPtr=/Game/Sounds/Fort_Step_Sounds/Soldier/Elevations/Upstairs/Soldier_Step_Walk_Metal_06_Upstairs2}
 }
 {
  id=WavePlayer_26
  properties {soundWaveAssetPtr=/Game/Sounds/Fort_Step_Sounds/Soldier/Elevations/Upstairs/Soldier_Step_Walk_Metal_07_Upstairs2}
 }
 {
  id=WavePlayer_27
  properties {soundWaveAssetPtr=/Game/Sounds/Fort_Step_Sounds/Soldier/Elevations/Upstairs/Soldier_Step_Walk_Metal_10_Upstairs2}
 }
 {
  id=WavePlayer_28
  properties {soundWaveAssetPtr=/Game/Sounds/Fort_Step_Sounds/Soldier/Elevations/Upstairs/Soldier_Step_Walk_Metal_12_Upstairs2}
 }
 {
  id=WavePlayer_34
  properties {soundWaveAssetPtr=&Soldier_Step_Walk_Metal_04_Upstairs2}
 }
 {
  id=WavePlayer_35
  properties {soundWaveAssetPtr=&Soldier_Step_Walk_Metal_11_Upstairs2}
 }
 {
  id=WavePlayer_36
  properties {soundWaveAssetPtr=&Soldier_Step_Walk_Metal_01_Upstairs2}
 }
]
properties {
 attenuationSettings=/Game/Sounds/Attenuation/Fort_Player_Footsteps_3P_Sprint_Athena_Att
 maxDistance=3584
 soundClassObject=&Fort_Player_Foley_3P_Enemy_Athena
 volumeMultiplier=0.9
 concurrencyOverrides {bLimitToOwner=true, maxCount=1, resolutionRule=StopOldest}
 concurrencySet [&Athena_Footstep_Local_Conc, &Athena_Footstep_Global_Conc]
}
