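# Sound cue for third-person soldier walking footsteps on metal from below, using "Downstairs" samples under the enemy foley sound class.
# A quality-level switch sends its first two slots to a random pick of twelve waves and its last two to a three-wave subset, each through a Modulator.
# Third-person walk attenuation applies, with local and global footstep concurrency sets plus a one-voice-per-owner, stop-oldest override.
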SoundCue /Game/Athena/Sounds/Character/Footsteps/3P/Metal/Athena_Soldier_Metal_Walk_Footstep_Below_3P_Cue
root=QualityLevel_0
nodes [
 {id=Modulator_0, children [Random_2]}
 {id=Modulator_1, children [Random_3]}
 {id=QualityLevel_0, children [Modulator_0,Modulator_0,Modulator_1,Modulator_1]}
 {
  id=Random_2
  children [
   WavePlayer_22
   WavePlayer_23
   WavePlayer_26
   WavePlayer_27
   WavePlayer_28
   WavePlayer_29
   WavePlayer_30
   WavePlayer_31
   WavePlayer_32
   WavePlayer_33
   WavePlayer_34
   WavePlayer_35
  ]
  properties {weights [1,1,1,1,1,1,1,1,1,1,1,1]}
 }
 {id=Random_3, children [WavePlayer_11, WavePlayer_12, WavePlayer_13], properties {bShouldExcludeFromBranchCulling=true, weights [1,1,1]}}
 {
  id=WavePlayer_11
  properties {soundWaveAssetPtr=/Game/Sounds/Fort_Step_Sounds/Soldier/Elevations/Downstairs/Soldier_Step_Walk_Metal_01_Downstairs}
 }
 {
  id=WavePlayer_12
  properties {soundWaveAssetPtr=/Game/Sounds/Fort_Step_Sounds/Soldier/Elevations/Downstairs/Soldier_Step_Walk_Metal_02_Downstairs}
 }
 {
  id=WavePlayer_13
  properties {soundWaveAssetPtr=/Game/Sounds/Fort_Step_Sounds/Soldier/Elevations/Downstairs/Soldier_Step_Walk_Metal_03_Downstairs}
 }
 {
  id=WavePlayer_22
  properties {soundWaveAssetPtr=&Soldier_Step_Walk_Metal_01_Downstairs}
 }
 {
  id=WavePlayer_23
  properties {soundWaveAssetPtr=&Soldier_Step_Walk_Metal_02_Downstairs}
 }
 {
  id=WavePlayer_26
  properties {soundWaveAssetPtr=&Soldier_Step_Walk_Metal_03_Downstairs}
 }
 {
  id=WavePlayer_27
  properties {soundWaveAssetPtr=/Game/Sounds/Fort_Step_Sounds/Soldier/Elevations/Downstairs/Soldier_Step_Walk_Metal_04_Downstairs}
 }
 {
  id=WavePlayer_28
  properties {soundWaveAssetPtr=/Game/Sounds/Fort_Step_Sounds/Soldier/Elevations/Downstairs/Soldier_Step_Walk_Metal_05_Downstairs}
 }
 {
  id=WavePlayer_29
  properties {soundWaveAssetPtr=/Game/Sounds/Fort_Step_Sounds/Soldier/Elevations/Downstairs/Soldier_Step_Walk_Metal_06_Downstairs}
 }
 {
  id=WavePlayer_30
  properties {soundWaveAssetPtr=/Game/Sounds/Fort_Step_Sounds/Soldier/Elevations/Downstairs/Soldier_Step_Walk_Metal_07_Downstairs}
 }
 {
  id=WavePlayer_31
  properties {soundWaveAssetPtr=/Game/Sounds/Fort_Step_Sounds/Soldier/Elevations/Downstairs/Soldier_Step_Walk_Metal_08_Downstairs}
 }
 {
  id=WavePlayer_32
  properties {soundWaveAssetPtr=/Game/Sounds/Fort_Step_Sounds/Soldier/Elevations/Downstairs/Soldier_Step_Walk_Metal_09_Downstairs}
 }
 {
  id=WavePlayer_33
  properties {soundWaveAssetPtr=/Game/Sounds/Fort_Step_Sounds/Soldier/Elevations/Downstairs/Soldier_Step_Walk_Metal_10_Downstairs}
 }
 {
  id=WavePlayer_34
  properties {soundWaveAssetPtr=/Game/Sounds/Fort_Step_Sounds/Soldier/Elevations/Downstairs/Soldier_Step_Walk_Metal_11_Downstairs}
 }
 {
  id=WavePlayer_35
  properties {soundWaveAssetPtr=/Game/Sounds/Fort_Step_Sounds/Soldier/Elevations/Downstairs/Soldier_Step_Walk_Metal_12_Downstairs}
 }
]
properties {
 attenuationSettings=/Game/Sounds/Attenuation/Fort_Player_Footsteps_3P_Walk_Athena_Att
 maxDistance=3584
 soundClassObject=&Fort_Player_Foley_3P_Enemy_Athena
 volumeMultiplier=0.85
 concurrencyOverrides {bLimitToOwner=true, maxCount=1, resolutionRule=StopOldest}
 concurrencySet [&Athena_Footstep_Local_Conc, &Athena_Footstep_Global_Conc]
}
